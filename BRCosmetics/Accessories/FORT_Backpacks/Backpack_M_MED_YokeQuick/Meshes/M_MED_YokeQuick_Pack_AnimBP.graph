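# Anim Blueprint for the YokeQuick backpack, whose suitcase reacts when an enemy player is eliminated.
# An anim input event briefly raises a flag that, with the Stage1 Parts variant enabled, interpolates Suitcase Alpha.
# That alpha drives additive suitcase and lock bone rotations, followed by AnimDynamics on the handle and small handle and pack offsets.
# In the front end, a looping timer replays the reaction and calls an FX dispatcher.
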AnimBlueprint /BRCosmetics/Accessories/FORT_Backpacks/Backpack_M_MED_YokeQuick/Meshes/M_MED_YokeQuick_Pack_AnimBP
target_skeleton=&/M_MED_YokeQuick_Pack_Skeleton
graphs [
 AnimGraph {nodes [
   n1 = Get Suitcase Alpha() -> (Suitcase Alpha:double)
   n2 = "Transform (Modify) Bone - Bone: dyn_suitcase1"(ComponentPose=Local To Component(LocalPose=Input Pose()), Alpha=@n1, Translation="0,0,0", Rotation="-7,0,0", Scale="1,1,1") {
    rotationMode=BMM_Additive
    rotationSpace=BCS_BoneSpace
    scaleMode=BMM_Additive
    translationMode=BMM_Additive
    translationSpace=BCS_BoneSpace
    alphaBoolBlend {blendOption=Linear}
    boneToModify {boneName=dyn_suitcase1}
    rotation {pitch=-7}
   }
   n3 = "Transform (Modify) Bone - Bone: dyn_lock_l_01"(ComponentPose=@n2, Alpha=@n1, Translation="0,0,0", Rotation="0,0,-30", Scale="1,1,1") {
    rotationMode=BMM_Additive
    rotationSpace=BCS_BoneSpace
    scaleMode=BMM_Additive
    translationMode=BMM_Additive
    translationSpace=BCS_BoneSpace
    alphaBoolBlend {blendOption=Linear}
    boneToModify {boneName=dyn_lock_l_01}
    rotation {roll=-30}
   }
   n4 = "Transform (Modify) Bone - Bone: dyn_lock_r_01"(ComponentPose=@n3, Alpha=@n1, Translation="0,0,0", Rotation="0,0,-25", Scale="1,1,1") {
    rotationMode=BMM_Additive
    rotationSpace=BCS_BoneSpace
    scaleMode=BMM_Additive
    translationMode=BMM_Additive
    translationSpace=BCS_BoneSpace
    alphaBoolBlend {blendOption=Linear}
    boneToModify {boneName=dyn_lock_r_01}
    rotation {roll=-25}
   }
   n5 = Anim Dynamics - Bone: dyn_handle_1!(ComponentPose=@n4, bAlphaBoolEnabled:bool=True, GravityOverride:Vector="0,0,-980", ExternalForce="0,0,0") -> (Pose) {
    alphaInputType=Bool
    angularDampingOverride=1
    angularSpringConstant=2000
    bAngularSpring=true
    bOverrideAngularDamping=true
    bUseGravityOverride=true
    gravityScale=0.35
    simulationSpace=RootRelative
    alphaBoolBlend {blendInTime=0.2, blendOption=Linear, blendOutTime=0.2}
    boundBone {boneName=dyn_handle_1}
    componentLinearAccScale=<1 1 1>
    componentLinearVelScale=<1 1 1>
    physicsBodyDefinitions [{
      boundBone {boneName=dyn_handle_1}
      boxExtents=<15 15 15>
      constraintSetup {angularLimitsMax=<0 45 0>, angularLimitsMin=<0 -45 0>, angularTarget=<1 0 0>}
      localJointOffset=<-1 0 0>
     }]
    relativeSpaceBone {boneName=dyn_handle_1}
   }
   n6 = "Transform (Modify) Bone - Bone: dyn_handle_1"!(ComponentPose=@n5, bAlphaBoolEnabled:bool=True, Translation="0,0,0", Rotation="35,0,0", Scale="1,1,1") -> (Pose) {
    alphaInputType=Bool
    rotationMode=BMM_Additive
    rotationSpace=BCS_BoneSpace
    scaleMode=BMM_Additive
    translationMode=BMM_Additive
    translationSpace=BCS_BoneSpace
    alphaBoolBlend {blendInTime=0.2, blendOption=Linear, blendOutTime=0.2}
    boneToModify {boneName=dyn_handle_1}
    rotation {pitch=35}
   }
   n7 = "Transform (Modify) Bone - Bone: pack"!(ComponentPose=@n6, bAlphaBoolEnabled:bool, Translation="0,-1,0", Rotation="0,0,0", Scale) -> (Pose) {
    alphaInputType=Bool
    rotationMode=BMM_Additive
    rotationSpace=BCS_BoneSpace
    scaleMode=BMM_Additive
    translationMode=BMM_Additive
    translationSpace=BCS_BoneSpace
    alphaBoolBlend {blendInTime=0.1, blendOption=Linear, blendOutTime=0.1}
    boneToModify {boneName=pack}
    translation {y=-1}
   }
   Output Pose(Result=Component To Local(ComponentPose=@n7)) {layerGroup=None}
  ]}
 EventGraph {nodes [
   n1 = Reactive() {is=title K2Node_CustomEvent}
   > Branch(Condition=Get bIsInFrontEnd())
   > Delay(Duration=2)
   > Call Yoke Quick FETrigger FX!(execute:Exec, self:M MED Yoke Quick Pack Anim BP Object Reference, EnemyPlayerEliminated:bool=true) -> (then:Exec) {is=title K2Node_CallDelegate}
   > Set bEnemyPlayerEliminated(bEnemyPlayerEliminated:bool=true) -> (Output_Get:bool)
   > Delay(Duration=4.5)
   > Set bEnemyPlayerEliminated(bEnemyPlayerEliminated=false)
   > Call Yoke Quick FETrigger FX()
   n2 = Event BlueprintUpdateAnimation()
   > Set Delta Time X(Delta Time X=@n2.DeltaTimeX)
   > Set NotInSkydiveMode(NotInSkydiveMode:bool=NOT Boolean(A=Get bIsSkydiveDiveMode())) -> (Output_Get:bool)
   > n3 = Branch(Condition=Get ReactivityEnabled() -> (ReactivityEnabled:bool))
   n4 = Get bEnemyPlayerEliminated() -> (bEnemyPlayerEliminated:bool)
   > Set Suitcase Alpha(Suitcase Alpha:double=FInterpTo(Current=Get Suitcase Alpha(), Target=SelectFloat(A=1, B=0, bPickA=@n4), DeltaTime=Get Delta Time X(), InterpSpeed=3)) -> (Output_Get:double)
   Set Suitcase Alpha(execute=@n3.else, Suitcase Alpha=0)
   Event OnCharacterCustomizationLoaded()
   > n5 = Cast To FortPlayerPawn(Object=GetOwningActor())
   > n6 = GetVariantChannelActiveIndexForItemActor!(
    execute
    Item:Athena Cosmetic Item Definition Object Reference=/BRCosmetics/Athena/Items/Cosmetics/Backpacks/Backpack_YokeQuick
    Actor:Actor Object Reference=@n5
    Channel:Gameplay Tag Structure=(TagName="Cosmetics.Variant.Channel.Parts")
    VariantList:Array of Gameplay Tag Structures=Make Array("[0]"=(TagName="Cosmetics.Variant.Property.Stage1"))
    bErrorIfChannelMissing:bool=false
    bErrorIfTagNotInArray:bool=false
   ) -> (then, Index:int=0, bIndexIsValid:bool, bFoundChannel:bool)
   > n7 = Branch(Condition=@n6.bIndexIsValid)
   > Set ReactivityEnabled(ReactivityEnabled:bool=true) -> (Output_Get:bool)
   > Reactive() {is=title K2Node_CallFunction}
   Set ReactivityEnabled(execute=@n7.else, ReactivityEnabled=false)
   n8 = Event BlueprintProcessAnimInputEvent()
   > Branch(Condition=IsNamed(self=@n8.AnimInputEvent, InEventName=aie.EnemyPlayerEliminated))
   > Set bEnemyPlayerEliminated(bEnemyPlayerEliminated=true)
   > Reactive()
   > Delay(Duration=4.5)
   > Set bEnemyPlayerEliminated(bEnemyPlayerEliminated=false)
   Event BlueprintBeginPlay() {is=title K2Node_Event}
   > Parent: BlueprintBeginPlay() {is=title K2Node_CallParentFunction}
   > Branch(Condition=IsCosmeticObjectDisplayedInFrontendContext!(CosmeticObject:Object Reference) -> (ReturnValue:bool))
   > Delay(Duration=4)
   > Reactive()
   > Set Timer by Event(Delegate=@n1, Time=11, bLooping=true)
   TryGetPawnOwner()
  ]}
 YokeQuickFETriggerFX {nodes [YokeQuickFETriggerFX!() -> (then, EnemyPlayerEliminated:bool=~) {is=title K2Node_FunctionEntry}]}
]
variables [
 bEnemyPlayerEliminated {type=boolean}
 Delta Time X {type=number}
 NotInSkydiveMode {type=boolean}
 ReactivityEnabled {type=boolean}
 Suitcase Alpha {type=number}
 YokeQuickFETriggerFX {default=()}
]
